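# Sound cue for crow wing-flap foley, played as a tactical creature-step sound with shared crow foley attenuation.
# A quality-level switch feeds the first two levels from a random pool of eight flaps and the last two from a pool of four, which is excluded from branch culling.
# A root modulator randomizes pitch slightly above normal.
# A concurrency override caps simultaneous instances.
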SoundCue /Irwin/Sounds/Crow/Foley/Crow_Flap_Cue
root=Modulator_0
nodes [
 {id=Modulator_0, children [QualityLevel_0], properties {pitchMax=1.2, pitchMin=1.1}}
 {id=QualityLevel_0, children [Random_1,Random_1,Random_3,Random_3]}
 {
  id=Random_1
  children [WavePlayer_12, WavePlayer_13, WavePlayer_14, WavePlayer_15, WavePlayer_16, WavePlayer_17, WavePlayer_18, WavePlayer_19]
  properties {weights [1,1,1,1,1,1,1,1]}
 }
 {
  id=Random_3
  children [WavePlayer_20, WavePlayer_21, WavePlayer_22, WavePlayer_23]
  properties {bShouldExcludeFromBranchCulling=true, weights [1,1,1,1]}
 }
 {id=WavePlayer_12, properties {soundWaveAssetPtr=&/Crow_Foley_WingFlap_01}}
 {id=WavePlayer_13, properties {soundWaveAssetPtr=&/Crow_Foley_WingFlap_02}}
 {id=WavePlayer_14, properties {soundWaveAssetPtr=&/Crow_Foley_WingFlap_03}}
 {id=WavePlayer_15, properties {soundWaveAssetPtr=&/Crow_Foley_WingFlap_04}}
 {id=WavePlayer_16, properties {soundWaveAssetPtr=&/Crow_Foley_WingFlap_05}}
 {id=WavePlayer_17, properties {soundWaveAssetPtr=&/Crow_Foley_WingFlap_06}}
 {id=WavePlayer_18, properties {soundWaveAssetPtr=&/Crow_Foley_WingFlap_07}}
 {id=WavePlayer_19, properties {soundWaveAssetPtr=&/Crow_Foley_WingFlap_08}}
 {id=WavePlayer_20, properties {soundWaveAssetPtr=&Crow_Foley_WingFlap_01}}
 {id=WavePlayer_21, properties {soundWaveAssetPtr=&Crow_Foley_WingFlap_02}}
 {id=WavePlayer_22, properties {soundWaveAssetPtr=&Crow_Foley_WingFlap_03}}
 {id=WavePlayer_23, properties {soundWaveAssetPtr=&Crow_Foley_WingFlap_04}}
]
properties {
 attenuationSettings=/Irwin/Sounds/Crow/Crow_Foley_Att
 bOverrideConcurrency=true
 maxDistance=2200
 soundClassObject=/Game/Sounds/FortSoundClassesAndModes/Fort_Creatures_Steps_Tactical
 volumeMultiplier=1
 concurrencyOverrides {maxCount=4}
}
