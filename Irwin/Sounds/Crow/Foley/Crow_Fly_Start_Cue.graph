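# Sound cue for a crow's take-off foley, played as it starts flying.
# A random node picks one of four equally weighted fly-start waves, then a modulator varies its pitch from normal to raised.
# It uses the crow foley attenuation and a creature-steps tactical sound class, with a slight volume boost and a concurrency override capping simultaneous instances.
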SoundCue /Irwin/Sounds/Crow/Foley/Crow_Fly_Start_Cue
root=Modulator_0
nodes [
 {id=Modulator_0, children [Random_0], properties {pitchMax=1.3, pitchMin=1}}
 {id=Random_0, children [WavePlayer_0,WavePlayer_1,WavePlayer_2,WavePlayer_3], properties {weights [1,1,1,1]}}
 {id=WavePlayer_0, properties {soundWaveAssetPtr=&/Crow_Foley_Fly_Start_01}}
 {id=WavePlayer_1, properties {soundWaveAssetPtr=&/Crow_Foley_Fly_Start_02}}
 {id=WavePlayer_2, properties {soundWaveAssetPtr=&/Crow_Foley_Fly_Start_03}}
 {id=WavePlayer_3, properties {soundWaveAssetPtr=&/Crow_Foley_Fly_Start_04}}
]
properties {
 attenuationSettings=/Irwin/Sounds/Crow/Crow_Foley_Att
 bOverrideConcurrency=true
 maxDistance=2200
 soundClassObject=/Game/Sounds/FortSoundClassesAndModes/Fort_Creatures_Steps_Tactical
 volumeMultiplier=1.2
 concurrencyOverrides {maxCount=4}
}
