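# Looping spatialized foley cue for crow flight, routed to the creature foley sound class with attenuation and a concurrency limit.
# A mixer layers a pitched-down crow flight loop with a pitched-up skydive parachute wind loop, each with slight pitch randomization.
# A continuous modulator on the mix maps the Flying parameter to pitch and volume, so the sound responds to that value at runtime.
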SoundCue /Irwin/Sounds/Crow/Foley/Crow_Flying_Loop_Cue
root=ModulatorContinuous_0
nodes [
 {id=Mixer_1, children [Modulator_1,Modulator_0], properties {inputVolume [1,1]}}
 {
  id=ModulatorContinuous_0
  children [Mixer_1]
  properties {
   pitchModulationParams {maxInput=2, maxOutput=2, minInput=0.8, minOutput=0.8, parameterName=Flying}
   volumeModulationParams {maxInput=1.2, maxOutput=1.2, minInput=0.8, minOutput=0.8, parameterName=Flying}
  }
 }
 {id=Modulator_0, children [WavePlayer_3], properties {pitchMax=1.55, pitchMin=1.45, volumeMax=1, volumeMin=1}}
 {id=Modulator_1, children [WavePlayer_1], properties {pitchMax=0.7, pitchMin=0.6, volumeMax=1, volumeMin=1}}
 {id=WavePlayer_1, properties {bLooping=true, soundWaveAssetPtr=&/Crow_Flight_Loop_01}}
 {id=WavePlayer_3, properties {bLooping=true, soundWaveAssetPtr=/Game/Athena/Sounds/Wind_Falling/sfx_parachute_skydive_loop_3p}}
]
properties {
 attenuationSettings=/Irwin/Sounds/Crow/Crow_Flight_Att
 bOverrideConcurrency=true
 maxDistance=2500
 soundClassObject=/Game/Sounds/FortSoundClassesAndModes/Fort_Creatures_Foley
 volumeMultiplier=0.5
 concurrencyOverrides {maxCount=2}
}
