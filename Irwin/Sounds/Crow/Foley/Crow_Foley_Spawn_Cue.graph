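# Crow spawn foley sound cue that reuses Nug spawn foley recordings for the Crow creature.
# A Random node picks one of four equally weighted waves, and a Modulator shifts it to a slightly raised, varied pitch.
# It uses Crow foley attenuation and the tactical creature steps sound class.
# A concurrency override allows one instance per owner and blocks new plays while one is active.
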SoundCue /Irwin/Sounds/Crow/Foley/Crow_Foley_Spawn_Cue
root=Modulator_1
nodes [
 {id=Modulator_1, children [Random_1], properties {pitchMax=1.15, pitchMin=1.05}}
 {id=Random_1, children [WavePlayer_0, WavePlayer_9, WavePlayer_10, WavePlayer_11], properties {weights [1,1,1,1]}}
 {id=WavePlayer_0, properties {soundWaveAssetPtr=/Irwin/Sounds/Nug/Foley/Nug_Foley_Spawn_010}}
 {id=WavePlayer_10, properties {soundWaveAssetPtr=/Irwin/Sounds/Nug/Foley/Nug_Foley_Spawn_030}}
 {id=WavePlayer_11, properties {soundWaveAssetPtr=/Irwin/Sounds/Nug/Foley/Nug_Foley_Spawn_040}}
 {id=WavePlayer_9, properties {soundWaveAssetPtr=/Irwin/Sounds/Nug/Foley/Nug_Foley_Spawn_020}}
]
properties {
 attenuationSettings=/Irwin/Sounds/Crow/Crow_Foley_Att
 bOverrideConcurrency=true
 maxDistance=2200
 soundClassObject=/Game/Sounds/FortSoundClassesAndModes/Fort_Creatures_Steps_Tactical
 volumeMultiplier=1
 concurrencyOverrides {bLimitToOwner=true, maxCount=1, resolutionRule=PreventNew}
}
