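# Looping sound cue for the crow's loot glow, routed to the creature foley sound class.
# A looping wave feeds an Enveloper that ramps volume up from near silence over about four seconds, then a Modulator that randomizes pitch slightly above normal.
# Cue-level overrides give it logarithmic distance falloff, low-pass and volume occlusion, and a tight concurrency limit.
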SoundCue /Irwin/Sounds/Crow/Foley/Crow_LootGlow_Loop_Cue
root=Modulator_2
nodes [
 {
  id=Enveloper_0
  children [WavePlayer_2]
  properties {pitchCurve {editorCurveData {keys [{time=0}]}}, volumeCurve {editorCurveData {keys [<at 0.0741 0.0084>,<at 4.2804 1>]}}}
 }
 {id=Modulator_2, children [Enveloper_0], properties {pitchMax=1.2, pitchMin=1.15, volumeMax=1, volumeMin=1}}
 {id=WavePlayer_2, properties {bLooping=true, soundWaveAssetPtr=&/Crow_LootGlow_Loop_01}}
]
properties {
 bOverrideAttenuation=true
 bOverrideConcurrency=true
 maxDistance=2500
 soundClassObject=/Game/Sounds/FortSoundClassesAndModes/Fort_Creatures_Foley
 volumeMultiplier=0.35
 attenuationOverrides {
  distanceAlgorithm=Logarithmic
  falloffDistance=2500
  occlusionInterpolationTime=1
  occlusionLowPassFilterFrequency=6000
  occlusionTraceChannel=ECC_GameTraceChannel11
  occlusionVolumeAttenuation=0.6
  reverbDistanceMax=1024
  reverbDistanceMin=0
  reverbWetLevelMax=0
  reverbWetLevelMin=1
  stereoSpread=500
  attenuationShapeExtents {x=0}
 }
 concurrencyOverrides {maxCount=2}
}
